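; BP_Burd_Window_A_A_A is a window Blueprint from the Burd set that can spawn an optional addon mesh.
; Its Construction Script calls the parent, then passes the Spawn Addon Mesh toggle, mesh, and material override settings to SpawnAddonMesh.
; As an exposed asset provider, it adds version 1 and reports the addon mesh and override material as indirect references to allow.
; BeginPlay, overlap and Tick hold only disabled parent calls.
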(event EventBeginPlay
  (|CallParentFunction) ; This node is disabled and will not be called.\nDrag off pins to build functionality.
)

(event Collision|EventActorBeginOverlap
  (params (OtherActor :type Actor%20Object%20Reference))
  (|CallParentFunction OtherActor) ; This node is disabled and will not be called.\nDrag off pins to build functionality.
)

(event EventTick
  (params (DeltaSeconds :type Float%20%28single-precision%29))
  (|CallParentFunction DeltaSeconds) ; This node is disabled and will not be called.\nDrag off pins to build functionality.
)

(fn BuildExposedAssetProviderVersion(Implementation)
  (params (Version :type Fort%20Exposed%20Asset%20Provider%20Version%20Structure%20%28by%20ref%29))
  (|CallParentFunction Version)
  (ExposedAssetProvider|AddVersion Version 1)
  (return))

(fn CollectAssetReferencestoAllow(Implementation)
  (params (Collector :type Fort%20Exposed%20Asset%20Reference%20Collector%20Structure%20%28by%20ref%29))
  (|CallParentFunction Collector)
  (ExposedAssetProvider|CollectIndirectAssetReference Collector (Variables|Default|GetAddon%20Mesh))
  (ExposedAssetProvider|CollectIndirectAssetReference Collector (Variables|Default|GetOverride%20Material))
  (return))

(fn ConstructionScript
  (|CallParentFunction)
  (AddonMeshOptions|SpawnAddonMesh (Variables|Default|GetSpawn%20Addon%20Mesh) (Variables|Default|GetAddon%20Mesh) self false (Variables|Default|GetOverride%20Addon%20Material) (Variables|Default|GetOverride%20Material)))
